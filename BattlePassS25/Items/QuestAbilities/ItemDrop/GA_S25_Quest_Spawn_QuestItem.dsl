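; Gameplay ability that spawns a loose Coral Samples quest actor for a Season 25 Coral story quest (Part 02).
; On activation it waits for damage gameplay events and checks whether the target's tags include the Coral P02 quest tag.
; A match has a roughly even random chance to spawn the drop through SpawnQuestActorsfromDropEntry, passing along the event data.
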(event Ability|EventActivateAbility
  (bind (_asynctaskproxy _payload) (Ability|Tasks|WaitGameplayEvent "(TagName=\"Event.Damage\")" 0 false false)
    (:EventReceived
      (bind (_payload_eventtag _payload_instigator _payload_target _payload_optionalobject _payload_optionalobject2 _payload_contexthandle _payload_instigatortags _payload_targettags _payload_eventmagnitude _payload_targetdata) (Utilities|Struct|BreakGameplayEventData _payload))
      (if (GameplayTags|HasTag _payload_targettags "(TagName=\"Athena.Quests.S25.Coral.P02\")")
        (if (<= (Math|Random|RandomFloat) 0.5)
          (Spawning|SpawnQuestActorsfromDropEntry 0 (struct :SpawnedQuestActorClass_6_F15CFDE2424412E9BADBE2975B317A07 "/BattlePassS25/Items/QuestActors/Stories/Part02/Quest_02/BP_S25_Loose_CoralSamples2.BP_S25_Loose_CoralSamples2_C" :DropQuantity_14_6469E6844F83772BB80669BCF263CB9D 1 :FailCount_10_49E24745476ACF5406DC15A79201F03E 1) (Utilities|Struct|MakeGameplayEventData _payload_eventtag _payload_instigator _payload_target _payload_optionalobject _payload_optionalobject2 _payload_contexthandle _payload_instigatortags _payload_targettags _payload_eventmagnitude _payload_targetdata)))))))

(event Ability|EventOnEndAbility
  (params (bWasCancelled :type Boolean)))
